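# Physics asset for the SleepyCloud pickaxe cosmetic, defining a jointed chain of dynamic-named bones plus character leg bodies.
# Six capsules on dyn_SleepyCloud_01 to 06 link in series from a tiny base_SleepyCloud capsule, with limited swing and tight twist, widening toward the tip.
# Thigh and calf capsules on both legs are joined by free-rotation constraints and have no constraint to the chain, suggesting they serve as collision shapes for it.
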PhysicsAsset /BRCosmetics/Weapons/FORT_Melee/Pickaxe_SleepyCloud/Meshes/SleepyCloud_Axe_PhysicsFE
bodies {
 base_SleepyCloud [{shapeType=Capsule, length=0.505, radius=0.505, shapeName=base_SleepyCloud_capsule, rotation=<rot 0 -90 0>}]
 calf_l [{shapeType=Capsule, length=30, radius=8, shapeName=thigh_l_capsule, center=<-26 0 -0.797>, rotation=<rot 90 180 0>}]
 calf_r [{shapeType=Capsule, length=30, radius=9.5, shapeName=thigh_r_capsule, center=<26 0 1.1417>, rotation=<rot 90 0 0>}]
 dyn_SleepyCloud_01 [{shapeType=Capsule, length=5, radius=4.5, shapeName=dyn_SleepyCloud_01_capsule, center=<1 0 0>, rotation=<rot 90 0 0>}]
 dyn_SleepyCloud_02 [{shapeType=Capsule, length=5, radius=4.5, shapeName=dyn_SleepyCloud_02_capsule, center=<1.5 0 0>, rotation=<rot 90 0 0>}]
 dyn_SleepyCloud_03 [{shapeType=Capsule, length=5, radius=4.5, shapeName=dyn_SleepyCloud_03_capsule, center=<2 0 0>, rotation=<rot 90 0 0>}]
 dyn_SleepyCloud_04 [{shapeType=Capsule, length=5, radius=7, shapeName=dyn_SleepyCloud_04_capsule, center=<2.5 0 0>, rotation=<rot 90 0 0>}]
 dyn_SleepyCloud_05 [{shapeType=Capsule, length=5, radius=7, shapeName=dyn_SleepyCloud_05_capsule, center=<3 3.5 0>, rotation=<rot 90 0 0>}]
 dyn_SleepyCloud_06 [{shapeType=Capsule, length=5, radius=7, shapeName=dyn_SleepyCloud_06_capsule, center=<3.5 3.5 0>, rotation=<rot 90 0 0>}]
 thigh_l [{shapeType=Capsule, length=34.6616, radius=9.5479, shapeName=thigh_l_capsule, center=<-22.1907 -0.6066 -0.5906>, rotation=<rot 90 180 0>}]
 thigh_r [{shapeType=Capsule, length=34.6616, radius=9.5479, shapeName=thigh_r_capsule, center=<22.1907 0.6066 0.9958>, rotation=<rot 90 0 0>}]
}
constraints [
 {bone1Name=dyn_SleepyCloud_01, bone2Name=base_SleepyCloud, swing1LimitDegrees=30, swing2LimitDegrees=25, twistLimitDegrees=2}
 {bone1Name=dyn_SleepyCloud_02, bone2Name=dyn_SleepyCloud_01, swing1LimitDegrees=20, swing2LimitDegrees=25, twistLimitDegrees=4}
 {bone1Name=dyn_SleepyCloud_03, bone2Name=dyn_SleepyCloud_02, swing1LimitDegrees=25, swing2LimitDegrees=30, twistLimitDegrees=4}
 {bone1Name=dyn_SleepyCloud_04, bone2Name=dyn_SleepyCloud_03, swing1LimitDegrees=25, swing2LimitDegrees=30, twistLimitDegrees=4}
 {bone1Name=dyn_SleepyCloud_05, bone2Name=dyn_SleepyCloud_04, swing1LimitDegrees=25, swing2LimitDegrees=30, twistLimitDegrees=4}
 {bone1Name=dyn_SleepyCloud_06, bone2Name=dyn_SleepyCloud_05, swing1LimitDegrees=25, swing2LimitDegrees=30, twistLimitDegrees=4}
 {
  bone1Name=calf_l
  bone2Name=thigh_l
  swing1LimitDegrees=45
  swing1Motion=Free
  swing2LimitDegrees=45
  swing2Motion=Free
  twistLimitDegrees=45
  twistMotion=Free
 }
 {
  bone1Name=calf_r
  bone2Name=thigh_r
  swing1LimitDegrees=45
  swing1Motion=Free
  swing2LimitDegrees=45
  swing2Motion=Free
  twistLimitDegrees=45
  twistMotion=Free
 }
]
